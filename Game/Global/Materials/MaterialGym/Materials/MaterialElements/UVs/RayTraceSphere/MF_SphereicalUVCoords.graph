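# Material function that converts a 3D position, local position by default, into spherical (equirectangular) UV coordinates.
# It normalizes the input to a direction, then derives U from a fast arctangent2 of Y over X, scaled to one turn, negated and offset to center.
# V comes from the arcsine of Z, scaled by pi, offset and inverted, and both combine into a float2 output.
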MaterialFunction /Game/Global/Materials/MaterialGym/Materials/MaterialElements/UVs/RayTraceSphere/MF_SphereicalUVCoords
graph {expressions {
  MaterialExpressionAdd_1 {
   class=MaterialExpressionAdd
   inputs [
    A {from=MaterialExpressionDivide_2}
    B {from=""}
   ]
   properties {constB=0.5}
  }
  MaterialExpressionAdd_7 {
   class=MaterialExpressionAdd
   inputs [
    A {from=MaterialExpressionMultiply_0}
    B {from=""}
   ]
   properties {constB=0.5}
  }
  MaterialExpressionArcsine_0 {class=MaterialExpressionArcsine, inputs [None {from=MaterialExpressionComponentMask_2, output=""}], properties {}}
  MaterialExpressionArctangent2Fast_0 {
   class=MaterialExpressionArctangent2Fast
   inputs [
    Y {from=MaterialExpressionComponentMask_1, output=""}
    X {from=MaterialExpressionComponentMask_0, output=""}
   ]
   properties {}
  }
  MaterialExpressionComponentMask_0 {class=MaterialExpressionComponentMask, inputs [None {from=MaterialExpressionNormalize_0}], properties {r=true}}
  MaterialExpressionComponentMask_1 {class=MaterialExpressionComponentMask, inputs [None {from=MaterialExpressionNormalize_0}], properties {g=true}}
  MaterialExpressionComponentMask_2 {class=MaterialExpressionComponentMask, inputs [None {from=MaterialExpressionNormalize_0}], properties {b=true}}
  MaterialExpressionConvert_0 {
   class=MaterialExpressionConvert
   inputs [
    None {from=MaterialExpressionAdd_7}
    None {from=MaterialExpressionOneMinus_0}
   ]
   properties {
    nodeName=Make Float2
    convertInputs [
     {
      type=Scalar
      expressionInput {
       expression=&self:MaterialExpressionAdd_7
       inputName=None
       mask=0
       maskA=0
       maskB=0
       maskG=0
       maskR=0
       outputIndex=0
      }
     }
     {
      type=Scalar
      expressionInput {
       expression=&self:MaterialExpressionOneMinus_0
       inputName=None
       mask=0
       maskA=0
       maskB=0
       maskG=0
       maskR=0
       outputIndex=0
      }
     }
    ]
    convertOutputs [{type=Vector2}]
   }
  }
  MaterialExpressionDivide_1 {
   class=MaterialExpressionDivide
   inputs [
    A {from=MaterialExpressionArctangent2Fast_0}
    B {from=""}
   ]
   properties {constB=6.28}
  }
  MaterialExpressionDivide_2 {
   class=MaterialExpressionDivide
   inputs [
    A {from=MaterialExpressionArcsine_0}
    B {from=""}
   ]
   properties {constB=3.141}
  }
  MaterialExpressionFunctionInput_0 {
   class=MaterialExpressionFunctionInput
   inputs [Preview {from=MaterialExpressionLocalPosition_0, output=XYZ}]
   properties {bUsePreviewValueAsDefault=true, inputName=Position}
  }
  MaterialExpressionFunctionOutput_0 {class=MaterialExpressionFunctionOutput, inputs [None {from=MaterialExpressionConvert_0, output=""}]}
  MaterialExpressionLocalPosition_0 {class=MaterialExpressionLocalPosition, inputs [], properties {}}
  MaterialExpressionMultiply_0 {
   class=MaterialExpressionMultiply
   inputs [
    A {from=MaterialExpressionDivide_1}
    B {from=""}
   ]
   properties {constB=-1}
  }
  MaterialExpressionNormalize_0 {class=MaterialExpressionNormalize, inputs [VectorInput {from=MaterialExpressionFunctionInput_0, output=""}]}
  MaterialExpressionOneMinus_0 {class=MaterialExpressionOneMinus, inputs [None {from=MaterialExpressionAdd_1}]}
 }}
