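# Material function that analytically ray-traces a virtual sphere per pixel, outputting front and back hit positions, normals, spherical UVs and a mask.
# It casts a ray from the camera through the pixel's world position against center and radius inputs, with center defaulting to actor position.
# Dot-product projection and square roots yield both intersection distances.
# If nodes build the mask and substitute a constant for position on one branch.
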MaterialFunction /Game/Global/Materials/MaterialGym/Materials/MaterialElements/UVs/RayTraceSphere/MF_RaytraceSphere
graph {nodes [
  FunctionOutput(None=NamedRerouteUsage()) {outputName=Frontface Pos}
  FunctionOutput(None=NamedRerouteUsage()) {outputName=Backface Pos, sortPriority=3}
  FunctionOutput(None=NamedRerouteUsage()) {outputName=Frontface UVs, sortPriority=2}
  FunctionOutput(None=NamedRerouteUsage()) {outputName=Backface UVs, sortPriority=5}
  FunctionOutput(None=NamedRerouteUsage()) {outputName=Mask, sortPriority=6}
  FunctionOutput(None=NamedRerouteUsage()) {outputName=Frontface Normal, sortPriority=1}
  FunctionOutput(None=NamedRerouteUsage()) {outputName=Backface Normal, sortPriority=4}
  n1 = NamedRerouteUsage()
  n2 = NamedRerouteUsage()
  n3 = Subtract(A=Add(A=NamedRerouteUsage(), B=Multiply(A=@n1, B=@n2)), B=NamedRerouteUsage())
  n4 = If(A=NamedRerouteUsage(), B=NamedRerouteUsage(), "A > B"=Constant() {r=1}, "A == B"=@n3, "A < B"=@n3)
  NamedRerouteDeclaration(None=@n4) {name=FrontFacePos, nodeColor=<rgba 0 0.7656 1 1>}
  NamedRerouteDeclaration(None=FunctionInput(Preview=ActorPositionWS()) {bUsePreviewValueAsDefault=true, inputName=center}) {name=center, nodeColor=<rgba 0 0.8359 1 1>}
  NamedRerouteDeclaration(None=Add(A=NamedRerouteUsage(), B=NamedRerouteUsage())) {name=t2, nodeColor=<rgba 0 1 0.6016 1>}
  n5 = NamedRerouteUsage()
  n6 = NamedRerouteUsage()
  n7 = Subtract(A=Add(A=NamedRerouteUsage(), B=Multiply(A=@n5, B=@n6)), B=NamedRerouteUsage())
  n8 = If(A=NamedRerouteUsage(), B=NamedRerouteUsage(), "A > B"=Constant() {r=1}, "A == B"=@n7, "A < B"=@n7)
  NamedRerouteDeclaration(None=@n8) {name=BackFacePos, nodeColor=<rgba 0 0.7656 1 1>}
  NamedRerouteDeclaration(None=MaterialFunctionCall!(Position=NamedRerouteUsage()) {materialFunction=&/MF_SphereicalUVCoords}.Result) {name=BackfaceUVs, nodeColor=<rgba 1 0 0.3281 1>}
  NamedRerouteDeclaration(None=MaterialFunctionCall(Position=NamedRerouteUsage()) {materialFunction=&MF_SphereicalUVCoords}.Result) {name=FrontFaceUVs, nodeColor=<rgba 0 0.6484 1 1>}
  n9 = Constant()
  NamedRerouteDeclaration(None=If(A=NamedRerouteUsage(), B=NamedRerouteUsage(), "A > B"=@n9, "A == B"=@n9, "A < B"=Constant() {r=1})) {name=Mask, nodeColor=<rgba 0.9688 1 0 1>}
  NamedRerouteDeclaration(None=Normalize(VectorInput=@n4)) {name=Frontface_Normal, nodeColor=<rgba 1 0 0.9141 1>}
  NamedRerouteDeclaration(None=Normalize(VectorInput=@n8)) {name=Backface_Normal, nodeColor=<rgba 1 0 0.9141 1>}
  NamedRerouteDeclaration(None=CameraPositionWS()) {name=ray_origin, nodeColor=<rgba 1 0.8438 0 1>}
  NamedRerouteDeclaration(None=Subtract(A=NamedRerouteUsage(), B=NamedRerouteUsage())) {name=L, nodeColor=<rgba 0.5234 0 1 1>}
  n10 = Normalize(VectorInput=Subtract(A=WorldPosition().XYZ, B=NamedRerouteUsage()))
  NamedRerouteDeclaration(None=DotProduct(A=NamedRerouteUsage(), B=NamedRerouteDeclaration(None=@n10) {name=ray, nodeColor=<rgba 0.9219 0 1 1>})) {name=tc, nodeColor=<rgba 0 1 0.2031 1>}
  n11 = MaterialFunctionCall(Vector 3=NamedRerouteUsage()) {materialFunction=/Engine/Functions/Engine_MaterialFunctions02/Utility/VectorLength}
  n12 = Multiply(A=@n11.V3 Length, B=@n11.V3 Length)
  n13 = NamedRerouteUsage()
  n14 = Multiply(A=@n13, B=@n13)
  NamedRerouteDeclaration(None=SquareRoot(None=Subtract(A=@n12, B=@n14))) {name=d, nodeColor=<rgba 1 0 0.1172 1>}
  n15 = FunctionInput() {bUsePreviewValueAsDefault=true, inputName=radius, inputType=FunctionInput_Scalar, previewValue=<100 0 0 1>}
  NamedRerouteDeclaration(None=@n15) {name=radius, nodeColor=<rgba 0 0.4375 1 1>}
  n16 = NamedRerouteUsage()
  n17 = Multiply(A=@n16, B=@n16)
  n18 = NamedRerouteUsage()
  n19 = Multiply(A=@n18, B=@n18)
  NamedRerouteDeclaration(None=SquareRoot(None=Subtract(A=@n17, B=@n19))) {name=t1c, nodeColor=<rgba 0.7578 1 0 1>}
  NamedRerouteDeclaration(None=Subtract(A=NamedRerouteUsage(), B=NamedRerouteUsage())) {name=t1, nodeColor=<rgba 0 1 0.6016 1>}
  NamedRerouteUsage()
 ]}
