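# M_Raytrace_Sphere_Back is a material gym example that maps a texture using the backface UVs of a ray-traced sphere.
# The MF_RaytraceSphere function takes a Radius parameter and supplies Backface UVs, which sample the UV editor color grid texture as a test pattern.
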Material /Game/Global/Materials/MaterialGym/Materials/MaterialElements/UVs/RayTraceSphere/M_Raytrace_Sphere_Back
graph {
 properties {bAutomaticallySetUsageInEditor=false, usedShadingModels=Default Lit}
 nodes [TextureSample(UVs=MaterialFunctionCall!(center, radius=ScalarParameter() {defaultValue=97.2611, parameterName=Radius}) {materialFunction=&/MF_RaytraceSphere}.Backface UVs) {texture=/UVEditor/Textures/UVEditorColorGrid_Color}]
}
