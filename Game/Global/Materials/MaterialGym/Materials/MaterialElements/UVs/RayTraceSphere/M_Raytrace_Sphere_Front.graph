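# Material Gym example material that renders the front face of a raytraced sphere through MF_RaytraceSphere, sized by a Radius parameter.
# The function's frontface UVs sample the UV editor color grid, with U scaled by parameter A.
# Its world-space frontface normal is converted to tangent space and blended with a flat normal using the sphere mask.
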Material /Game/Global/Materials/MaterialGym/Materials/MaterialElements/UVs/RayTraceSphere/M_Raytrace_Sphere_Front
graph {
 expressions {
  MaterialExpressionConstant3Vector_0 {class=MaterialExpressionConstant3Vector, properties {constant=<rgba 0 0 1 0>}}
  MaterialExpressionConvert_0 {
   class=MaterialExpressionConvert
   inputs [None {from=MaterialExpressionMaterialFunctionCall_2, output=Frontface UVs}]
   properties {
    nodeName=Break Float2
    convertInputs [{
      type=Vector2
      expressionInput {
       expression=&self:MaterialExpressionMaterialFunctionCall_2
       inputName=None
       mask=0
       maskA=0
       maskB=0
       maskG=0
       maskR=0
       outputIndex=2
      }
     }]
    convertOutputs [{type=Scalar}, {type=Scalar}]
   }
  }
  MaterialExpressionConvert_1 {
   class=MaterialExpressionConvert
   inputs [
    None {from=MaterialExpressionMultiply_1}
    None {from=MaterialExpressionConvert_0}
   ]
   properties {
    nodeName=Make Float2
    convertInputs [
     {
      type=Scalar
      expressionInput {
       expression=&self:MaterialExpressionMultiply_1
       inputName=None
       mask=0
       maskA=0
       maskB=0
       maskG=0
       maskR=0
       outputIndex=0
      }
     }
     {
      type=Scalar
      expressionInput {
       expression=&self:MaterialExpressionConvert_0
       inputName=None
       mask=0
       maskA=0
       maskB=0
       maskG=0
       maskR=0
       outputIndex=1
      }
     }
    ]
    convertOutputs [{type=Vector2}]
   }
  }
  MaterialExpressionLinearInterpolate_0 {
   class=MaterialExpressionLinearInterpolate
   inputs [
    A {from=MaterialExpressionConstant3Vector_0}
    B {from=MaterialExpressionTransform_1}
    Alpha {from=MaterialExpressionMaterialFunctionCall_2, output=Mask}
   ]
   properties {constAlpha=1, constB=0}
  }
  MaterialExpressionMaterialFunctionCall_2 {
   class=MaterialExpressionMaterialFunctionCall
   inputs [
    center {from=""}
    radius {from=MaterialExpressionScalarParameter_5}
   ]
   properties {materialFunction=&/MF_RaytraceSphere}
  }
  MaterialExpressionMultiply_1 {
   class=MaterialExpressionMultiply
   inputs [
    A {from=MaterialExpressionScalarParameter_0}
    B {from=MaterialExpressionConvert_0}
   ]
  }
  MaterialExpressionScalarParameter_0 {class=MaterialExpressionScalarParameter, properties {defaultValue=1, parameterName=A}}
  MaterialExpressionScalarParameter_5 {class=MaterialExpressionScalarParameter, properties {defaultValue=50, parameterName=Radius}}
  MaterialExpressionTextureSample_1 {
   class=MaterialExpressionTextureSample
   inputs [
    UVs {from=MaterialExpressionConvert_1}
    Tex {from=""}
    Apply View MipBias {from=""}
   ]
   properties {texture=/UVEditor/Textures/UVEditorColorGrid_Color}
  }
  MaterialExpressionTransform_1 {
   class=MaterialExpressionTransform
   inputs [None {from=MaterialExpressionMaterialFunctionCall_2, output=Frontface Normal}]
   properties {transformSourceType=TRANSFORMSOURCE_World, transformType=TRANSFORM_Tangent}
  }
 }
 properties {bAutomaticallySetUsageInEditor=false, usedShadingModels=Default Lit}
}
